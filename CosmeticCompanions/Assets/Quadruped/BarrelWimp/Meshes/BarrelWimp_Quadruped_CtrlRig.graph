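# Control Rig for the BarrelWimp quadruped cosmetic companion that adjusts its skirt bones in response to rear leg motion.
# A sequence runs a left and a right branch, each with a SphericalPoseReader measuring a rear thigh's orientation relative to the hips.
# Each reader's output weights a ModifyTransforms on that side's base and dynamic skirt bones, acting as a pose-driven corrective.
# The hierarchy also carries face, IK and attach bones plus pose curves, none referenced by the visible graph.
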ControlRigBlueprint /CosmeticCompanions/Assets/Quadruped/BarrelWimp/Meshes/BarrelWimp_Quadruped_CtrlRig
graphs [
 RigVMFunctionLibrary {links [], nodes []}
 RigVMFunction_Sequence_1_ContainedGraph {nodes [
   n1 = RigVMFunctionEntryNode()
   RigVMFunction_Sequence = &RigVMFunction_Sequence(ExecuteContext=@n1.ExecuteContext)
   n2 = &RigVMFunction_Sequence(ExecuteContext=@RigVMFunction_Sequence.B)
   RigVMFunctionReturnNode(A=@RigVMFunction_Sequence.A, B=@n2.A, C=@n2.B)
  ]}
 RigVMModel {nodes [
   RigUnit_BeginExecution = &RigUnit_BeginExecution()
   RigVMFunction_Sequence = RigVMAggregateNode(ExecuteContext=@RigUnit_BeginExecution.ExecutePin)
   n1 = &RigUnit_SphericalPoseReader(
    ActiveRegionScaleFactors=(PositiveWidth=1,NegativeWidth=1,PositiveHeight=1,NegativeHeight=1)
    ActiveRegionSize=0.02
    Debug=(bDrawDebug=false,bDraw2D=False,bDrawLocalAxes=False,DebugScale=15,DebugSegments=20,DebugThickness=0.25)
    DriverAxis=(X=1,Y=0,Z=0)
    DriverItem=(Type=Bone,Name="thigh_bk_l")
    FalloffRegionScaleFactors=(PositiveWidth=1,NegativeWidth=1,PositiveHeight=1,NegativeHeight=1)
    FalloffSize=0.221
    FlipHeightScaling=False
    FlipWidthScaling=False
    OptionalParentItem=(Type=Bone,Name="hips")
    RotationOffset=(X=55,Y=-60,Z=0)
    ExecutePin=@RigVMFunction_Sequence.A
   )
   > &RigUnit_ModifyTransforms(
    ItemToModify="((Item=(Type=Bone,Name=\"base_skirt_l_01\"),Transform=(Rotation=(X=0,Y=0,Z=0.0698,W=0.9976),Translation=(X=0,Y=0,Z=0),Scale3D=(X=1,Y=1,Z=1))),(Item=(Type=Bone,Name=\"dyn_skirt_l_01\"),Transform=(Rotation=(X=0.0781,Y=-0.3483,Z=-0.3648,W=-0.8599),Translation=(X= ...(truncated)"
    WeightMinimum=0
    Weight=@n1.OutputParam
   )
   n2 = &RigUnit_SphericalPoseReader(
    ActiveRegionScaleFactors=(PositiveWidth=1,NegativeWidth=1,PositiveHeight=1,NegativeHeight=1)
    ActiveRegionSize=0.02
    Debug=(bDrawDebug=false,bDraw2D=False,bDrawLocalAxes=False,DebugScale=15,DebugSegments=20,DebugThickness=0.25)
    DriverAxis=(X=1,Y=0,Z=0)
    DriverItem=(Type=Bone,Name="thigh_bk_r")
    FalloffRegionScaleFactors=(PositiveWidth=1,NegativeWidth=1,PositiveHeight=1,NegativeHeight=1)
    FalloffSize=0.221
    FlipHeightScaling=False
    FlipWidthScaling=False
    OptionalParentItem=(Type=Bone,Name="hips")
    RotationOffset=(X=55,Y=-60,Z=0)
    ExecutePin=@RigVMFunction_Sequence.B
   )
   > &RigUnit_ModifyTransforms(
    ItemToModify="((Item=(Type=Bone,Name=\"base_skirt_r_01\"),Transform=(Rotation=(X=0,Y=0,Z=-0.1285,W=-0.9917),Translation=(X=0,Y=0,Z=0),Scale3D=(X=1,Y=1,Z=1))),(Item=(Type=Bone,Name=\"dyn_skirt_r_01\"),Transform=(Rotation=(X=-0.3693,Y=-0.0607,Z=0.1784,W=0.91),Translation=(X=1 ...(truncated)"
    WeightMinimum=0
    Weight=@n2.OutputParam
   )
  ]}
]
hierarchy [
 root {loc=<0 0 0>, rot=<0 0 0 1>}
 pelvis {parent=root, loc=<0 -9.1937 28.2113>, rot=<0.5131 0.4865 -0.5131 -0.4865>}
 spine_01 {parent=pelvis, loc=<0 0 0>, rot=<0 0 -0.0124 -0.9999>}
 spine_02 {parent=spine_01, loc=<11.0133 -0.2742 0>, rot=<0 0 0.014 -0.9999>}
 spine_03 {parent=spine_02, loc=<9.8251 0 0>, rot=<0 0 0.031 -0.9995>}
 neck_01 {parent=spine_03, loc=<0.7833 0.3499 0>, rot=<0 0 0.5768 -0.8169>}
 neck_02 {parent=neck_01, loc=<6.4826 0 0>, rot=<0 0 -0.0432 -0.9991>}
 head {parent=neck_02, loc=<4.1335 -0.0047 0>, rot=<-0.0138 -0.707 0.0138 -0.707>}
 face_attach {parent=head, loc=<0 0.2264 2.1556>, rot=<-0.175 0 0 -0.9846>}
 jaw {parent=face_attach, loc=<0 0.6245 0.3131>, rot=<0 0.7071 0 -0.7071>}
 lip_lower_03_l {parent=jaw, loc=<-2.2831 9.2755 -1.2161>, rot=<-0.4858 -0.702 -0.3803 -0.3558>}
 lip_lower_02_l {parent=jaw, loc=<-2.112 9.5054 -0.8729>, rot=<-0.5539 -0.635 -0.3633 -0.3974>}
 lip_lower_01_l {parent=jaw, loc=<-1.8189 9.5812 -0.4585>, rot=<0.5821 0.5772 0.427 0.3817>}
 lip_lower_mid {parent=jaw, loc=<-1.7613 9.7149 0>, rot=<0.4939 0.506 0.506 0.4939>}
 lip_lower_01_r {parent=jaw, loc=<-1.8189 9.5812 0.4585>, rot=<0.5772 -0.5821 -0.3817 0.427>}
 lip_lower_02_r {parent=jaw, loc=<-2.112 9.5054 0.8729>, rot=<0.635 -0.5539 -0.3974 0.3633>}
 lip_lower_03_r {parent=jaw, loc=<-2.2831 9.2755 1.2161>, rot=<0.702 -0.4858 -0.3558 0.3803>}
 teeth_lower {parent=jaw, loc=<-2.8213 4.0295 0>, rot=<-0.003 -0.7071 0.003 -0.7071>}
 tongue_01 {parent=jaw, loc=<-2.5163 1.1457 0>, rot=<0.4148 -0.5726 0.5726 -0.4148>}
 tongue_02 {parent=tongue_01, loc=<-2.6682 0 0.1879>, rot=<0 0.1656 0 -0.9862>}
 tongue_03 {parent=tongue_02, loc=<3.9963 0 0.0753>, rot=<0 0.0015 0 -1>}
 lip_corner_l {parent=face_attach, loc=<1.5984 9.8933 -2.0505>, rot=<0.2428 -0.3246 0.0168 -0.914>}
 cheek_bt_l {parent=face_attach, loc=<5.7493 9.3479 -2.796>, rot=<0.0001 -0.0158 -0.6457 -0.7635>}
 cheek_bt_r {parent=face_attach, loc=<-5.7493 9.3479 -2.796>, rot=<0.7635 -0.6457 0.0158 0.0001>}
 reverse_jaw {parent=face_attach, loc=<0 0.0433 0.7402>, rot=<0.0006 0.7071 -0.0006 -0.7071>}
 lip_upper_mid {parent=reverse_jaw, loc=<-1.7954 10.3931 0>, rot=<-0.4829 -0.5166 -0.5166 -0.4829>}
 lip_upper_01_l {parent=reverse_jaw, loc=<-2.0567 10.3328 -0.6357>, rot=<-0.5978 -0.6488 -0.3885 -0.2659>}
 lip_upper_02_l {parent=reverse_jaw, loc=<-2.2366 10.1423 -0.9871>, rot=<-0.5597 -0.6498 -0.38 -0.3465>}
 lip_upper_03_l {parent=reverse_jaw, loc=<-2.4717 9.9713 -1.3578>, rot=<-0.5528 -0.7039 -0.3417 -0.2866>}
 cheek_tp_in_l {parent=reverse_jaw, loc=<-0.3647 10.6284 -3.5549>, rot=<0.5112 0.5064 0.5027 0.479>}
 cheek_tp_ot_l {parent=reverse_jaw, loc=<-0.4927 9.419 -6.1052>, rot=<-0.3984 -0.5833 -0.4258 -0.5655>}
 cheek_tp_in_r {parent=reverse_jaw, loc=<-0.3647 10.6284 3.5549>, rot=<-0.5064 0.5112 0.479 -0.5027>}
 cheek_tp_ot_r {parent=reverse_jaw, loc=<-0.4927 9.419 6.1052>, rot=<0.5833 -0.3984 -0.5655 0.4258>}
 brow_ot_r {parent=reverse_jaw, loc=<4.2189 6.9343 8.5012>, rot=<0.5867 0.3946 -0.5861 -0.3956>}
 brow_md_r {parent=reverse_jaw, loc=<5.5584 8.5046 6.2437>, rot=<0.6794 0.196 -0.6791 -0.1971>}
 brow_ct {parent=reverse_jaw, loc=<6.3461 10.5578 0>, rot=<-0.0006 -0.7071 0.0006 -0.7071>}
 brow_md_l {parent=reverse_jaw, loc=<5.5584 8.5046 -6.2437>, rot=<0.196 -0.6794 0.1971 -0.6791>}
 brow_in_l {parent=reverse_jaw, loc=<5.9983 9.9616 -3.1491>, rot=<0.0189 -0.7012 0.0218 -0.7123>}
 brow_ot_l {parent=reverse_jaw, loc=<4.2189 6.9343 -8.5012>, rot=<0.3946 -0.5867 0.3956 -0.5861>}
 brow_in_r {parent=reverse_jaw, loc=<5.9983 9.9616 3.1491>, rot=<-0.7012 -0.0189 0.7123 0.0218>}
 eye_r {parent=reverse_jaw, loc=<2.8936 5.3689 4.2916>, rot=<-0.698 -0.1081 0.7007 0.1012>}
 lid_bt_r {parent=reverse_jaw, loc=<2.8936 5.3689 4.2916>, rot=<-0.6941 -0.1205 0.7042 0.0887>}
 lid_tp_r {parent=reverse_jaw, loc=<2.8936 5.3689 4.2916>, rot=<-0.6941 -0.1205 0.7042 0.0887>}
 nose_mid {parent=reverse_jaw, loc=<0.6001 11.3148 0>, rot=<-0.5463 -0.4489 -0.4489 -0.5463>}
 nostril_l {parent=nose_mid, loc=<-0.1256 -1.1783 -0.2912>, rot=<0 0 0 -1>}
 nostril_r {parent=nose_mid, loc=<-0.1256 1.1783 -0.2912>, rot=<0 -1 0 0>}
 lid_tp_l {parent=reverse_jaw, loc=<2.8936 5.3689 -4.2916>, rot=<0.1205 -0.6941 0.0887 -0.7042>}
 lid_bt_l {parent=reverse_jaw, loc=<2.8936 5.3689 -4.2916>, rot=<0.1205 -0.6941 0.0887 -0.7042>}
 eye_l {parent=reverse_jaw, loc=<2.8936 5.3689 -4.2916>, rot=<0.1081 -0.698 0.1012 -0.7007>}
 lip_upper_01_r {parent=reverse_jaw, loc=<-2.0567 10.3328 0.6357>, rot=<0.6488 -0.5978 -0.2659 0.3885>}
 lip_upper_02_r {parent=reverse_jaw, loc=<-2.2366 10.1423 0.9871>, rot=<0.6498 -0.5597 -0.3465 0.38>}
 lip_upper_03_r {parent=reverse_jaw, loc=<-2.4717 9.9713 1.3578>, rot=<0.7039 -0.5528 -0.2866 0.3417>}
 teeth_upper {parent=reverse_jaw, loc=<0.1308 4.4464 0>, rot=<-0.0034 -0.7071 0.0034 -0.7071>}
 ear_01_l {parent=reverse_jaw, loc=<3.5458 1.5802 -2.3435>, rot=<-0.0419 -0.2529 0.1211 -0.959>}
 ear_02_l {parent=ear_01_l, loc=<10.6008 0 0>, rot=<0.5989 -0.5778 -0.4956 -0.2488>}
 ear_03_l {parent=ear_02_l, loc=<6.3081 0 0>, rot=<0.0373 -0.0039 -0.2633 -0.964>}
 ear_04_l {parent=ear_03_l, loc=<4.9436 0 0>, rot=<0.0094 0.0335 -0.025 -0.9991>}
 ear_01_r {parent=reverse_jaw, loc=<3.5458 1.5802 2.3435>, rot=<-0.2529 0.0419 0.959 0.1211>}
 ear_02_r {parent=ear_01_r, loc=<-10.6008 0 0>, rot=<0.5989 -0.5778 -0.4956 -0.2488>}
 ear_03_r {parent=ear_02_r, loc=<-6.3081 0 0>, rot=<0.0373 -0.0039 -0.2633 -0.964>}
 ear_04_r {parent=ear_03_r, loc=<-4.9436 0 0>, rot=<0.0094 0.0335 -0.025 -0.9991>}
 base_pomPom_l {parent=reverse_jaw, loc=<14.5947 0.1688 -8.0347>, rot=<0.6689 -0.2575 -0.3749 -0.588>}
 dyn_pomPom_l {parent=base_pomPom_l, loc=<0 0 0>, rot=<0 0 0 1>}
 base_pomPom_r {parent=reverse_jaw, loc=<14.5949 0.1692 8.035>, rot=<-0.3749 0.588 0.6689 0.2575>}
 dyn_pomPom_r {parent=base_pomPom_r, loc=<0 0 0>, rot=<0 0 0 1>}
 lip_corner_r {parent=face_attach, loc=<-1.5984 9.8933 -2.0505>, rot=<0.914 0.0168 0.3246 0.2428>}
 shoulder_fr_l {parent=spine_03, loc=<2.147 1.488 -6.748>, rot=<-0.0045 0.0076 -0.7951 -0.6064>}
 elbow_fr_l {parent=shoulder_fr_l, loc=<11.9351 0 0>, rot=<0.0041 0.0009 0.0951 -0.9955>}
 wrist_fr_l {parent=elbow_fr_l, loc=<8.5313 0 0>, rot=<0.9818 -0.1889 0.0185 -0.0005>}
 ball_fr_l {parent=wrist_fr_l, loc=<3.3 0 0>, rot=<-0.0537 0.0764 -0.3272 -0.9403>}
 toe_fr_l {parent=ball_fr_l, loc=<1.4732 0 0>, rot=<-0.0077 -0.039 -0.1557 -0.987>}
 toe_tip_fr_l {parent=toe_fr_l, loc=<3.5769 0 0>, rot=<0.003 -0.0026 -0.121 -0.9926>}
 scapula_fr_l {parent=shoulder_fr_l, loc=<-5.2648 4.2357 0.0731>, rot=<-0.0004 -0.0056 0.3323 -0.9432>}
 shoulder_fr_r {parent=spine_03, loc=<2.147 1.488 6.748>, rot=<0.0076 0.0045 0.6064 -0.7951>}
 elbow_fr_r {parent=shoulder_fr_r, loc=<-11.9351 0 -0.0001>, rot=<0.0041 0.0009 0.0951 -0.9955>}
 wrist_fr_r {parent=elbow_fr_r, loc=<-8.5313 0 0>, rot=<0.9818 -0.1889 0.0185 -0.0005>}
 ball_fr_r {parent=wrist_fr_r, loc=<-3.3 0 0>, rot=<-0.0537 0.0764 -0.3272 -0.9403>}
 toe_fr_r {parent=ball_fr_r, loc=<-1.4732 0 0>, rot=<-0.0077 -0.039 -0.1557 -0.987>}
 toe_tip_fr_r {parent=toe_fr_r, loc=<-3.5768 0 0.0001>, rot=<0.003 -0.0026 -0.121 -0.9926>}
 scapula_fr_r {parent=shoulder_fr_r, loc=<5.2648 -4.2358 -0.0731>, rot=<-0.0004 -0.0056 0.3323 -0.9432>}
 hips {parent=pelvis, loc=<0 0 0>, rot=<0 0 0 1>}
 thigh_bk_l {parent=hips, loc=<-3.1481 -1.1439 -4.4054>, rot=<0.7145 0.6991 -0.0205 0.0189>}
 knee_bk_l {parent=thigh_bk_l, loc=<12.9068 0 0>, rot=<0.0191 -0.009 0.3401 -0.9402>}
 ankle_bk_l {parent=knee_bk_l, loc=<9.5129 0 0>, rot=<-0.0174 0.0083 -0.3824 -0.9238>}
 ball_bk_l {parent=ankle_bk_l, loc=<6.9302 0 0>, rot=<-0.0127 -0.0228 -0.4083 -0.9125>}
 toe_bk_l {parent=ball_bk_l, loc=<1.8166 0 0>, rot=<0 0 -0.1426 -0.9898>}
 toe_tip_bk_l {parent=toe_bk_l, loc=<3.4203 0 0>, rot=<0 0 -0.2352 -0.9719>}
 thigh_bk_r {parent=hips, loc=<-3.148 -1.1439 4.4054>, rot=<0.6991 -0.7145 -0.0189 -0.0205>}
 knee_bk_r {parent=thigh_bk_r, loc=<-12.9068 0 0>, rot=<0.0191 -0.009 0.3401 -0.9402>}
 ankle_bk_r {parent=knee_bk_r, loc=<-9.5129 0 0>, rot=<-0.0174 0.0083 -0.3824 -0.9238>}
 ball_bk_r {parent=ankle_bk_r, loc=<-6.9302 0 0>, rot=<-0.0127 -0.0228 -0.4083 -0.9125>}
 toe_bk_r {parent=ball_bk_r, loc=<-1.8166 0 0>, rot=<0 0 -0.1426 -0.9898>}
 toe_tip_bk_r {parent=toe_bk_r, loc=<-3.4203 0 0>, rot=<0 0 -0.2352 -0.9719>}
 tail_01 {parent=hips, loc=<-7.5814 0.4536 -0.1448>, rot=<0.0095 -0.6637 0.707 -0.2439>}
 tail_02 {parent=tail_01, loc=<4.5941 0 0>, rot=<-0.4347 0.2813 -0.0957 -0.8502>}
 tail_03 {parent=tail_02, loc=<4.4146 0 0>, rot=<0.3097 0.536 -0.244 -0.7464>}
 tail_04 {parent=tail_03, loc=<4.4956 0 0>, rot=<0.0054 0.1501 -0.0138 -0.9886>}
 tail_05 {parent=tail_04, loc=<4.5754 0 0>, rot=<0.0875 -0.0375 -0.0084 -0.9954>}
 base_skirt_l_01 {parent=hips, loc=<-3.1481 -1.1439 -4.4054>, rot=<0.9524 0.1707 -0.2507 -0.0298>}
 dyn_skirt_l_01 {parent=base_skirt_l_01, loc=<9.997 0 0>, rot=<0.0631 -0.213 -0.2426 -0.9443>}
 base_skirt_r_01 {parent=hips, loc=<-3.1481 -1.144 4.4054>, rot=<0.9524 0.1707 0.2507 0.0298>}
 dyn_skirt_r_01 {parent=base_skirt_r_01, loc=<9.997 -0.0001 0>, rot=<-0.0631 0.213 -0.2426 -0.9443>}
 IK_Foot_Rear_Root {parent=root, loc=<0 -19.464 0>, rot=<0 -0.7824 0.6227 0>}
 IK_Foot_Rear_L {parent=IK_Foot_Rear_Root, loc=<-5.8444 -2.3606 -1.446>, rot=<0.6373 0.3064 -0.6373 -0.3064>}
 IK_Foot_Rear_R {parent=IK_Foot_Rear_Root, loc=<5.8444 -2.3606 -1.446>, rot=<-0.3064 0.6373 0.3064 -0.6373>}
 IK_Foot_Front_Root {parent=root, loc=<0 1.5578 0>, rot=<0.105 0 0 -0.9945>}
 IK_Foot_Front_L {parent=IK_Foot_Front_Root, loc=<6.1906 8.9391 4.4477>, rot=<-0.4489 -0.5618 -0.3412 -0.6054>}
 IK_Foot_Front_R {parent=IK_Foot_Front_Root, loc=<-6.1906 8.9391 4.4477>, rot=<0.6054 -0.3412 0.5618 -0.4489>}
 attach_emote_01 {parent=root, loc=<0 0 0>, rot=<0 0 0 1>}
 attach_emote_02 {parent=root, loc=<0 0 0>, rot=<0 0 0 1>}
 attach_emote_03 {parent=root, loc=<0 0 0>, rot=<0 0 0 1>}
 attach {parent=root, loc=<0 0 0>, rot=<0 0 0 1>}
 AnimateAcc {type=Curve, loc=<0 0 0>, rot=<0 0 0 1>}
 AttachEmote1InPelvisSpace {type=Curve, loc=<0 0 0>, rot=<0 0 0 1>}
 AttachEmote2InPelvisSpace {type=Curve, loc=<0 0 0>, rot=<0 0 0 1>}
 AttachEmote3InPelvisSpace {type=Curve, loc=<0 0 0>, rot=<0 0 0 1>}
 base_pose {type=Curve, loc=<0 0 0>, rot=<0 0 0 1>}
 Body_Scale {type=Curve, loc=<0 0 0>, rot=<0 0 0 1>}
 BodyMorph_LOD0 {type=Curve, loc=<0 0 0>, rot=<0 0 0 1>}
 BodyScale {type=Curve, loc=<0 0 0>, rot=<0 0 0 1>}
 BothLegIK {type=Curve, loc=<0 0 0>, rot=<0 0 0 1>}
 C_nose_wrinkler_pose {type=Curve, loc=<0 0 0>, rot=<0 0 0 1>}
 eye_convergence_pose {type=Curve, loc=<0 0 0>, rot=<0 0 0 1>}
 HoodieMorph_LOD0 {type=Curve, loc=<0 0 0>, rot=<0 0 0 1>}
 jaw_back_pose {type=Curve, loc=<0 0 0>, rot=<0 0 0 1>}
 jaw_forward_pose {type=Curve, loc=<0 0 0>, rot=<0 0 0 1>}
 jaw_left_pose {type=Curve, loc=<0 0 0>, rot=<0 0 0 1>}
 jaw_open_pose {type=Curve, loc=<0 0 0>, rot=<0 0 0 1>}
 jaw_right_pose {type=Curve, loc=<0 0 0>, rot=<0 0 0 1>}
 jaw_up_pose {type=Curve, loc=<0 0 0>, rot=<0 0 0 1>}
 L_blink_pose {type=Curve, loc=<0 0 0>, rot=<0 0 0 1>}
 L_brow_In_Dn_pose {type=Curve, loc=<0 0 0>, rot=<0 0 0 1>}
 L_brow_In_Up_pose {type=Curve, loc=<0 0 0>, rot=<0 0 0 1>}
 L_brow_Ot_Dn_pose {type=Curve, loc=<0 0 0>, rot=<0 0 0 1>}
 L_brow_Ot_Up_pose {type=Curve, loc=<0 0 0>, rot=<0 0 0 1>}
 L_brow_squeeze_in_pose {type=Curve, loc=<0 0 0>, rot=<0 0 0 1>}
 L_cheek_up_pose {type=Curve, loc=<0 0 0>, rot=<0 0 0 1>}
 L_eye_look_down_pose {type=Curve, loc=<0 0 0>, rot=<0 0 0 1>}
 L_eye_look_left_pose {type=Curve, loc=<0 0 0>, rot=<0 0 0 1>}
 L_eye_look_right_pose {type=Curve, loc=<0 0 0>, rot=<0 0 0 1>}
 L_eye_look_up_pose {type=Curve, loc=<0 0 0>, rot=<0 0 0 1>}
 L_frown_pose {type=Curve, loc=<0 0 0>, rot=<0 0 0 1>}
 L_lip_corner_narrow_pose {type=Curve, loc=<0 0 0>, rot=<0 0 0 1>}
 L_lip_corner_wide_pose {type=Curve, loc=<0 0 0>, rot=<0 0 0 1>}
 L_lower_lid_up_pose {type=Curve, loc=<0 0 0>, rot=<0 0 0 1>}
 L_lower_lip_down_pose {type=Curve, loc=<0 0 0>, rot=<0 0 0 1>}
 L_lower_lip_up_pose {type=Curve, loc=<0 0 0>, rot=<0 0 0 1>}
 L_smile_open_correction_pose {type=Curve, loc=<0 0 0>, rot=<0 0 0 1>}
 L_smile_pose {type=Curve, loc=<0 0 0>, rot=<0 0 0 1>}
 L_tongue_twist_pose {type=Curve, loc=<0 0 0>, rot=<0 0 0 1>}
 L_upper_lip_lower_pose {type=Curve, loc=<0 0 0>, rot=<0 0 0 1>}
 L_upper_lip_raiser_pose {type=Curve, loc=<0 0 0>, rot=<0 0 0 1>}
 L_upperLidDown_pose {type=Curve, loc=<0 0 0>, rot=<0 0 0 1>}
 L_wide_pose {type=Curve, loc=<0 0 0>, rot=<0 0 0 1>}
 lower_lip_dn_pose {type=Curve, loc=<0 0 0>, rot=<0 0 0 1>}
 lower_lip_up_pose {type=Curve, loc=<0 0 0>, rot=<0 0 0 1>}
 R_blink_pose {type=Curve, loc=<0 0 0>, rot=<0 0 0 1>}
 R_brow_In_Dn_pose {type=Curve, loc=<0 0 0>, rot=<0 0 0 1>}
 R_brow_In_Up_pose {type=Curve, loc=<0 0 0>, rot=<0 0 0 1>}
 R_brow_Ot_Dn_pose {type=Curve, loc=<0 0 0>, rot=<0 0 0 1>}
 R_brow_Ot_Up_pose {type=Curve, loc=<0 0 0>, rot=<0 0 0 1>}
 R_brow_squeeze_in_pose {type=Curve, loc=<0 0 0>, rot=<0 0 0 1>}
 R_cheek_up_pose {type=Curve, loc=<0 0 0>, rot=<0 0 0 1>}
 R_eye_look_down_pose {type=Curve, loc=<0 0 0>, rot=<0 0 0 1>}
 R_eye_look_left_pose {type=Curve, loc=<0 0 0>, rot=<0 0 0 1>}
 R_eye_look_right_pose {type=Curve, loc=<0 0 0>, rot=<0 0 0 1>}
 R_eye_look_up_pose {type=Curve, loc=<0 0 0>, rot=<0 0 0 1>}
 R_frown_pose {type=Curve, loc=<0 0 0>, rot=<0 0 0 1>}
 R_lip_corner_narrow_pose {type=Curve, loc=<0 0 0>, rot=<0 0 0 1>}
 R_lip_corner_wide_pose {type=Curve, loc=<0 0 0>, rot=<0 0 0 1>}
 R_lower_lid_up_pose {type=Curve, loc=<0 0 0>, rot=<0 0 0 1>}
 R_lower_lip_down_pose {type=Curve, loc=<0 0 0>, rot=<0 0 0 1>}
 R_lower_lip_up_pose {type=Curve, loc=<0 0 0>, rot=<0 0 0 1>}
 R_smile_open_correction_pose {type=Curve, loc=<0 0 0>, rot=<0 0 0 1>}
 R_smile_pose {type=Curve, loc=<0 0 0>, rot=<0 0 0 1>}
 R_tongue_twist_pose {type=Curve, loc=<0 0 0>, rot=<0 0 0 1>}
 R_upper_lip_lower_pose {type=Curve, loc=<0 0 0>, rot=<0 0 0 1>}
 R_upper_lip_raiser_pose {type=Curve, loc=<0 0 0>, rot=<0 0 0 1>}
 R_upperLidDown_pose {type=Curve, loc=<0 0 0>, rot=<0 0 0 1>}
 R_wide_pose {type=Curve, loc=<0 0 0>, rot=<0 0 0 1>}
 Reaction_Happy {type=Curve, loc=<0 0 0>, rot=<0 0 0 1>}
 reverse_jaw_pose {type=Curve, loc=<0 0 0>, rot=<0 0 0 1>}
 tongue_back_pose {type=Curve, loc=<0 0 0>, rot=<0 0 0 1>}
 tongue_curl_mid_down_pose {type=Curve, loc=<0 0 0>, rot=<0 0 0 1>}
 tongue_curl_mid_up_pose {type=Curve, loc=<0 0 0>, rot=<0 0 0 1>}
 tongue_curl_tip_down_pose {type=Curve, loc=<0 0 0>, rot=<0 0 0 1>}
 tongue_curl_tip_up_pose {type=Curve, loc=<0 0 0>, rot=<0 0 0 1>}
 tongue_down_pose {type=Curve, loc=<0 0 0>, rot=<0 0 0 1>}
 tongue_forward_pose {type=Curve, loc=<0 0 0>, rot=<0 0 0 1>}
 tongue_left_pose {type=Curve, loc=<0 0 0>, rot=<0 0 0 1>}
 tongue_right_pose {type=Curve, loc=<0 0 0>, rot=<0 0 0 1>}
 tongue_up_pose {type=Curve, loc=<0 0 0>, rot=<0 0 0 1>}
 upper_lip_dn_pose {type=Curve, loc=<0 0 0>, rot=<0 0 0 1>}
 upper_lip_up_pose {type=Curve, loc=<0 0 0>, rot=<0 0 0 1>}
]
