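# Post-process Anim Blueprint for the BarrelWimp quadruped cosmetic companion, adding pom-pom secondary motion and facial poses.
# The input pose runs through two chained AnimDynamics nodes for the right and left pom-pom bones, using damped angular springs, tight limits and reduced gravity.
# A DynamicsAlpha value scales both simulations.
# The result feeds a head facial PoseAsset blend node before output; the event graph holds only unconnected stub nodes.
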AnimBlueprint /CosmeticCompanions/Assets/Quadruped/BarrelWimp/Meshes/BarrelWimp_Quadruped_PostProcess_AnimBP
target_skeleton=&/BarrelWimp_Quadruped_Skeleton
graphs [
 AnimGraph {nodes [
   n1 = Get DynamicsAlpha() -> (DynamicsAlpha:float)
   n2 = Anim Dynamics - Bone: dyn_pomPom_r(ComponentPose=Local To Component(LocalPose=Input Pose()), Alpha=@n1, ExternalForce="0,0,0") {
    angularDampingOverride=0.9999
    angularSpringConstant=15000
    bAngularSpring=true
    bOverrideAngularDamping=true
    bUseGravityOverride=true
    gravityScale=0.35
    simulationSpace=RootRelative
    boundBone {boneName=dyn_pomPom_r}
    componentLinearAccScale=<1 1 1>
    componentLinearVelScale=<1 1 1>
    gravityOverride {z=-980}
    physicsBodyDefinitions [{
      boundBone {boneName=dyn_pomPom_r}
      boxExtents=<40 40 40>
      constraintSetup {angularLimitsMax=<0 10 10>, angularLimitsMin=<0 -10 -10>, angularTarget=<1 0 0>}
      localJointOffset=<4 0 0>
     }]
   }
   n3 = Anim Dynamics - Bone: dyn_pomPom_l(ComponentPose=@n2, Alpha=@n1, ExternalForce="0,0,0") {
    angularDampingOverride=0.9999
    angularSpringConstant=15000
    bAngularSpring=true
    bOverrideAngularDamping=true
    bUseGravityOverride=true
    gravityScale=0.35
    simulationSpace=RootRelative
    boundBone {boneName=dyn_pomPom_l}
    componentLinearAccScale=<1 1 1>
    componentLinearVelScale=<1 1 1>
    gravityOverride {z=-980}
    physicsBodyDefinitions [{
      boundBone {boneName=dyn_pomPom_l}
      boxExtents=<40 40 40>
      constraintSetup {angularLimitsMax=<0 10 10>, angularLimitsMin=<0 -10 -10>, angularTarget=<1 0 0>}
      localJointOffset=<4 0 0>
     }]
   }
   Output Pose(Result=BarrelWimp_Quadruped_Head_Facial_Poses_PoseAsset!(SourcePose:Pose Link Structure=Component To Local(ComponentPose=@n3)) -> (Pose:Pose Link Structure) {
     is=title AnimGraphNode_PoseBlendNode
     poseAsset=&/BarrelWimp_Quadruped_Head_Facial_Poses_PoseAsset
    })
  ]}
 EventGraph {nodes [
   TryGetPawnOwner()
   Event BlueprintUpdateAnimation()
  ]}
]
variables [NewVar {type=boolean}]
